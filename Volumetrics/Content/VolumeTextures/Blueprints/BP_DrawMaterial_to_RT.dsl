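; Blueprint actor that renders a material into a render target, likely for volume texture content given its folder.
; Its Draw event clears the render target, then draws the assigned material into it.
; The construction script runs Draw once, and Event Tick repeats it every frame while the Tick boolean is enabled.
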(event EventTick
  (params (DeltaSeconds :type Float%20%28single-precision%29))
  (if (Variables|Default|GetTick)
    (CallFunction|Draw)))

(event Custom|Clear
  (Rendering|ClearRenderTarget2D (Variables|Default|GetRender%20Target)))

(event Custom|Draw
  (CallFunction|Clear)
  (Rendering|DrawMaterialtoRenderTarget (Variables|Default|GetRender%20Target) (Variables|Default|GetMaterial)))

(fn ConstructionScript
  (CallFunction|Draw))
