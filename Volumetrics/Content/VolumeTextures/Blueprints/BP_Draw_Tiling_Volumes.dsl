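; BP_Draw_Tiling_Volumes is an editor tool actor that generates tiling noise volume textures as 2D pseudo-volume render targets.
; Construction creates draw and raymarch-preview MIDs, allocates an RGBA16f target sized from the volume resolution, and passes frame layout, tiling, octaves and persistence to the noise material.
; It then draws the noise into the target, which the mesh's raymarch material displays.
; A custom event saves the target as a static texture asset.
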(event Custom|Create%20Static%20Texture
  (EditorScripting|Rendering|RenderTarget2DCreateStaticTexture2DEditorOnly (Variables|Default|GetRT) (Utilities|String|Append (Variables|Default|GetNew%20Texture%20Path) (Variables|Default|GetNew%20Texture%20Name)) "TC_VectorDisplacementmap"))

(event Custom|UpdatePreview
  (bind _rt (Variables|Default|GetRT))
  (Rendering|ClearRenderTarget2D _rt)
  (Rendering|DrawMaterialtoRenderTarget _rt (Variables|Debug|GetDraw%20Material%20MID)))

(event Custom|Allocate%20RT
  (Class|VolumetricsBPLIbrary|CreateAndAssignRT (Variables|Default|GetRT) (Class|VolumetricsBPLIbrary|PseudoVolumeDimensions (Variables|Default|GetVolume%20Resolution)) "RTF_RGBA16f"))

(fn ConstructionScript
  (bind (_rt_size _xy_frames) (Class|VolumetricsBPLIbrary|PseudoVolumeDimensions (Variables|Default|GetVolume%20Resolution)))
  (bind _draw_material_mid (Variables|Debug|GetDraw%20Material%20MID))
  (Class|VolumetricsBPLIbrary|CreateAndAssignMID (Variables|Debug|GetDraw%20Material%20MID) (Variables|Default|GetNoise%20Generation%20Material))
  (Class|VolumetricsBPLIbrary|CreateAndAssignMID (Variables|Debug|GetRaymarch%20Preview%20MID) (Variables|Default|GetRaymarch%20Preview%20Material))
  (Rendering|Material|SetMaterial (Variables|Default|GetStaticMesh) 0 (Variables|Debug|GetRaymarch%20Preview%20MID))
  (CallFunction|AllocateRT)
  (Rendering|Material|SetVectorParameterValue (Variables|Debug|GetDraw%20Material%20MID) "XYFrames" (Utilities|IntPointToLinearColor _xy_frames) :!class /Script/Engine.MaterialInstanceDynamic)
  (Rendering|Material|SetTextureParameterValue (Variables|Debug|GetRaymarch%20Preview%20MID) "RT" (Variables|Default|GetRT))
  (Rendering|Material|SetScalarParameterValue _draw_material_mid "Tiling" (Variables|Default|GetTiling) :!class /Script/Engine.MaterialInstanceDynamic)
  (Rendering|Material|SetScalarParameterValue _draw_material_mid "Octaves" (Variables|Default|GetOctaves) :!class /Script/Engine.MaterialInstanceDynamic)
  (Rendering|Material|SetScalarParameterValue _draw_material_mid "Persistence" (Variables|Default|GetPersistence) :!class /Script/Engine.MaterialInstanceDynamic)
  (CallFunction|UpdatePreview))
