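; BP_PreviewVolume is a Blueprint that previews a parameterized material by painting it into a render target, apparently for volume texture work.
; Construction creates a dynamic paint material instance and runs UpdatePreview, which Tick repeats when its flag is enabled.
; UpdatePreview remaps transform P's location from ±128 units to a 0–1 vector, and derives scalars R and N from its X and Y scale.
; It then clears RT and draws the material into it.
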(event EventTick
  (params (DeltaSeconds :type Float%20%28single-precision%29))
  (if (Variables|Default|GetTick)
    (CallFunction|UpdatePreview)))

(event Custom|UpdatePreview
  (bind _p (Variables|Default|GetP))
  (bind _scale (.scale _p))
  (bind _rt (Variables|Default|GetRT))
  (bind _paintmid (Variables|Default|GetPaintMID))
  (Rendering|Material|SetVectorParameterValue _paintmid "P" (Math|Conversions|ToLinearColor(Vector) (* (+ (/ (.location _p) 128) 1) 0.5)) :!class /Script/Engine.MaterialInstanceDynamic)
  (Rendering|Material|SetScalarParameterValue _paintmid "R" (* (.x _scale) 0.25) :!class /Script/Engine.MaterialInstanceDynamic)
  (Rendering|Material|SetScalarParameterValue _paintmid "N" (* (.y _scale) 1) :!class /Script/Engine.MaterialInstanceDynamic)
  (Rendering|ClearRenderTarget2D _rt)
  (Rendering|DrawMaterialtoRenderTarget _rt (Variables|Default|GetPaintMID)))

(fn ConstructionScript
  (bind _returnvalue (Rendering|Material|CreateDynamicMaterialInstance (Variables|Default|GetMaterial) "None" "None" :!class /Script/Engine.KismetMaterialLibrary))
  (Variables|Default|SetPaintMID _returnvalue)
  (CallFunction|UpdatePreview))
