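# M_TestUVs is an unlit test material for sampling volume textures; none of its nodes connect to a material output.
# EncodeVolumeCoordinates feeds a position, using the XYFrames grid, into VolumeTextureFunction, which reads the Velocity Voronoi noise texture as a flipbook-style volume.
# A scalar parameter blends that result with a plain 2D sample of the Velocity parameter.
# A separate branch samples VT_CurlNoise at world position divided by a scale parameter.
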Material /Volumetrics/Content/VolumeTextures/Textures/M_TestUVs
graph {
 properties {pixelDepthOffsetMode=PDOM_Legacy, shadingModel=MSM_Unlit, usedShadingModels=Unlit}
 nodes [
  n1 = TextureSampleParameter2D() {
   desc=Velocity
   parameterName=Velocity
   samplerType=SAMPLERTYPE_LinearColor
   texture=&/T_Tiling_VoronoiNoise_256
  }
  ComponentMask(None=@n1.RGB) {b=true, g=true, r=true}
  n2 = TextureObjectParameter() {
   desc=Velocity
   parameterName=Velocity
   samplerType=SAMPLERTYPE_LinearColor
   texture=&T_Tiling_VoronoiNoise_256
  }
  n3 = ComponentMask(None=VectorParameter() {parameterName=XYFrames, defaultValue=<rgba 16 16 0 0>}.RGB) {g=true, r=true}
  LinearInterpolate(
   A=MaterialFunctionCall!(
    Texture Object=@n2
    Position=MaterialFunctionCall!(UVs, XY Frames=@n3) {materialFunction=/Volumetrics/Content/VolumeTextures/MaterialFunctions/EncodeVolumeCoordinates}.Result
    Scale
    Num Frames
    Clamp Z
   ) {materialFunction=/Volumetrics/Content/VolumeTextures/MaterialFunctions/VolumeTextureFunction}.Result RGB
   B=@n1.RGB
   Alpha=ScalarParameter() {parameterName=rrrr}
  )
  TextureSample(UVs=Divide(A=WorldPosition().XYZ, B=ScalarParameter() {defaultValue=100, parameterName=T})) {samplerType=SAMPLERTYPE_LinearColor, texture=&/VT_CurlNoise}
 ]
}
